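# Physics asset for the Future Samurai backpack cosmetic, defining collision bodies and joints for the pack, its belt pieces and two weapons.
# Boxes cover the pack and belt bones, while capsules cover the left and right weapons.
# The lower belt forms a four-link chain from the pack, and both upper belt pieces attach directly to it.
# Tight angular limits restrain the belts, with each weapon allowed only slight swing on one axis.
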PhysicsAsset /BRCosmetics/Accessories/FORT_Backpacks/Backpack_M_MED_Future_Samurai/Meshes/M_MED_Future_Samurai_Pack_Physics
bodies {
 dyn_belt_dn01 [{shapeType=Box, extentX=3.6781, extentY=0.7645, extentZ=4.3603}]
 dyn_belt_dn02 [{shapeType=Box, extentX=3.6781, extentY=0.7645, extentZ=4.3603}]
 dyn_belt_dn03 [{shapeType=Box, extentX=3.6781, extentY=0.7645, extentZ=4.3603}]
 dyn_belt_dn04 [{shapeType=Box, extentX=5.3785, extentY=1.7643, extentZ=4.5802, center=<0 0.3681 -0.9194>}]
 dyn_belt_up01 [{shapeType=Box, extentX=1.9563, extentY=0.9984, extentZ=5.1647, center=<0 0 1.9049>}]
 dyn_belt_up02 [{shapeType=Box, extentX=2.009, extentY=0.9984, extentZ=6.3994, center=<0 0 1.2605>}]
 pack [{shapeType=Box, extentX=20.8993, extentY=4.6093, extentZ=45.2657, center=<0.0001 3.6434 -18.2>, rotation=<rot 0 0 0.559>}]
 weapon_L [{shapeType=Capsule, length=39.7699, radius=5.3578, center=<1.908 0.1923 4.0934>, rotation=<rot 0.4288 -90 36.3073>}]
 weapon_R [{shapeType=Capsule, length=39.77, radius=5.358, center=<-2.3467 0.1944 3.0621>, rotation=<rot 0.4303 -90 -36.3108>}]
}
constraints [
 {bone1Name=dyn_belt_dn01, bone2Name=pack, swing1LimitDegrees=10, swing2LimitDegrees=10, twistLimitDegrees=20}
 {bone1Name=dyn_belt_dn02, bone2Name=dyn_belt_dn01, swing1LimitDegrees=10, swing2LimitDegrees=10, twistLimitDegrees=20}
 {bone1Name=dyn_belt_dn03, bone2Name=dyn_belt_dn02, swing1LimitDegrees=10, swing2LimitDegrees=10, twistLimitDegrees=20}
 {bone1Name=dyn_belt_dn04, bone2Name=dyn_belt_dn03, swing1LimitDegrees=0, swing2LimitDegrees=0, twistLimitDegrees=20}
 {bone1Name=dyn_belt_up01, bone2Name=pack, swing1LimitDegrees=10, swing2LimitDegrees=10, twistLimitDegrees=20}
 {bone1Name=dyn_belt_up02, bone2Name=pack, swing1LimitDegrees=10, swing2LimitDegrees=10, twistLimitDegrees=20}
 {bone1Name=weapon_L, bone2Name=pack, swing1LimitDegrees=0, swing2LimitDegrees=3, twistLimitDegrees=0}
 {bone1Name=weapon_R, bone2Name=pack, swing1LimitDegrees=0, swing2LimitDegrees=3, twistLimitDegrees=0}
]
