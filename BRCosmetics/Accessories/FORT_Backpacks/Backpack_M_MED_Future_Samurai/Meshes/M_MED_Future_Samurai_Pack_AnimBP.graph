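# Animation Blueprint for the Future Samurai backpack, adding physics, posture-driven offsets and material-based element hiding.
# Two chained RigidBody passes simulate in pack-bone space, then the pack bone is offset and scaled.
# Spine bend relative to pelvis drives additive pack and lower-belt rotations, weighted by AimDownSight, which actually sums moving, sprinting and skydive-floating flags.
# Weapon, customization and skydiving events set "Hide Element 01" from the equipped FutureSamurai-tagged pickaxe and skydive state.
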AnimBlueprint /BRCosmetics/Accessories/FORT_Backpacks/Backpack_M_MED_Future_Samurai/Meshes/M_MED_Future_Samurai_Pack_AnimBP
target_skeleton=&/M_MED_Future_Samurai_Pack_Skeleton
graphs [
 AnimGraph {nodes [
   Local Space Ref Pose()
   n1 = RigidBody(ComponentPose=Local To Component(LocalPose=Input Pose())) {
    bClampLinearTranslationLimitToRefPose=true
    bOverrideWorldGravity=true
    overridePhysicsAsset=&/M_MED_Future_Samurai_Pack_Physics1
    simulationSpace=BaseBoneSpace
    alphaBoolBlend {blendOption=Linear}
    baseBoneRef {boneName=pack}
    componentAppliedLinearAccClamp=<1000 1000 1000>
    componentLinearAccScale=<0.35 0.45 0.45>
    componentLinearVelScale=<0.3 0.3 0.3>
    overrideWorldGravity {z=-980}
   }
   n2 = RigidBody(ComponentPose=@n1) {
    bClampLinearTranslationLimitToRefPose=true
    bOverrideWorldGravity=true
    overridePhysicsAsset=&/M_MED_Future_Samurai_Pack_Physics
    simulationSpace=BaseBoneSpace
    alphaBoolBlend {blendOption=Linear}
    baseBoneRef {boneName=pack}
    componentAppliedLinearAccClamp=<750 750 750>
    componentLinearAccScale=<0.35 0.45 0.45>
    componentLinearVelScale=<0.3 0.3 0.3>
    overrideWorldGravity {z=-980}
   }
   n3 = "Transform (Modify) Bone - Bone: pack"(ComponentPose=@n2, Translation="0,-1,1.5", Rotation="0,0,0", Scale="0.95,0.95,0.95") {
    rotationMode=BMM_Additive
    rotationSpace=BCS_BoneSpace
    scaleMode=BMM_Replace
    translationMode=BMM_Additive
    translationSpace=BCS_BoneSpace
    alphaBoolBlend {blendOption=Linear}
    boneToModify {boneName=pack}
    scale=<0.95 0.95 0.95>
    translation {y=-1, z=1.5}
   }
   n4 = Get AimDownSight() -> (AimDownSight:double)
   n5 = "Transform (Modify) Bone - Bone: pack"(ComponentPose=@n3, Alpha=@n4, Translation="0,0,0", Rotation=Get backpack_rotate_01() -> (backpack_rotate_01:Rotator)) {
    rotationMode=BMM_Additive
    rotationSpace=BCS_BoneSpace
    translationMode=BMM_Additive
    translationSpace=BCS_BoneSpace
    alphaBoolBlend {blendOption=Linear}
    boneToModify {boneName=pack}
   }
   n6 = "Transform (Modify) Bone - Bone: dyn_belt_dn01"!(ComponentPose=@n5, Alpha=@n4, Translation="0,0,0", Rotation=Get backpack_rotate()) -> (Pose) {
    rotationMode=BMM_Additive
    rotationSpace=BCS_BoneSpace
    translationMode=BMM_Additive
    translationSpace=BCS_BoneSpace
    alphaBoolBlend {blendOption=Linear}
    alphaScaleBiasClamp {bInterpResult=true}
    boneToModify {boneName=dyn_belt_dn01}
   }
   Output Pose(Result=Component To Local(ComponentPose=@n6)) {layerGroup=None}
  ]}
 EventGraph {nodes [
   n1 = Event OnWeaponChanged()
   n2 = Get Dynamic Material Instance(execute=[@n1.then, @n10.then], SkeletalMeshComponent=GetPartSkeletalMeshComponent())
   > n3 = Sequence()
   n4 = PickaxeItemDefinitionHasAnyMetaTag(Tags=Make Array("[0]"=(TagName="Cosmetics.MetaTags.FutureSamurai")), bRequiresEquipped=true)
   > n5 = Branch(Condition=@n4)
   > n6 = Branch(Condition=Get bIsSkyDiving())
   SetScalarParameterValue(execute=[@n8.then, @n6.then, @n7.else, @n5.else], self=@n2, ParameterName=Hide Element 01, Value=0)
   SetScalarParameterValue(execute=[@n6.else, @n7.then], self=@n2, ParameterName=Hide Element 01, Value=1)
   n7 = Branch(execute=[@n9.then, @n3.then_1], Condition=@n4)
   n8 = Event OnBeginSkydiving()
   n9 = Event OnEndSkydiving()
   n10 = Event OnCharacterCustomizationLoaded()
   Event BlueprintUpdateAnimation()
   n11 = "To Integer (Boolean)"!(InBool:bool=Get bIsSkydiveFloating()) -> (ReturnValue:int=0)
   n12 = int + int(
    A="To Integer (Boolean)"(InBool=Get bIsMoving2D()) -> (ReturnValue=0)
    B="To Integer (Boolean)"(InBool=Get bIsSprinting()) -> (ReturnValue=0)
   ) -> (ReturnValue=0) {is=title K2Node_CommutativeAssociativeBinaryOperator}
   > Set AimDownSight(AimDownSight:double="To Float (Integer)"(InInt=int + int(A=@n12, B=@n11) -> (ReturnValue=0))) -> (Output_Get:double)
   n13 = GetOwnerSkeletalMeshComponent()
   n14 = BreakRotator(InRot=BreakTransform(InTransform=GetDeltaTransformFromRefPose(self=@n13, BoneName=spine_01, BaseName=pelvis)).Rotation)
   > Set backpack_rotate(backpack_rotate=MakeRotator(Roll=MapRangeClamped(Value=@n14, InRangeA=20, InRangeB=50, OutRangeA=20, OutRangeB=50)))
   > Set backpack_rotate_01(backpack_rotate_01:Rotator=MakeRotator(Roll=MapRangeClamped(Value=@n14, InRangeA=20, InRangeB=50, OutRangeA=1, OutRangeB=5))) -> (Output_Get:Rotator)
   TryGetPawnOwner()
  ]}
]
variables [
 AimDownSight {type=number}
 backpack_rotate {type=Rotator}
 backpack_rotate_01 {type=Rotator}
]
